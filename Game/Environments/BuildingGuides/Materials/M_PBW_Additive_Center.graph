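# Unlit additive material for building guide visuals, with a slowly pulsing brightness.
# A time-driven sine, remapped to blend between full and dimmed levels, multiplies the color texture parameter.
# One sum adds that pulsing color to a constant pale blue-grey tint masked by the red and blue channels of the alpha blend texture, leaving the tint steady.
# A second chain clamps the pulsed color and multiplies it by the pulse again.
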Material /Game/Environments/BuildingGuides/Materials/M_PBW_Additive_Center
graph {
 properties {blendMode=BLEND_Additive, pixelDepthOffsetMode=PDOM_Legacy, shadingModel=MSM_Unlit, usedShadingModels=Unlit}
 nodes [
  n1 = Constant3Vector() {constant=<rgba 0.2946 0.3044 0.355 1>}
  n2 = TextureSampleParameter2D() {parameterName=alpha blend, texture=/Game/Environments/BuildingGuides/Textures/T_PBW_BG_BalconyO_QuarterWallS_QuarterWallHalf_BP_01_M}
  n3 = Sine(None=Multiply(A=Constant() {r=0.5}, B=Time()))
  n4 = Constant() {r=2}
  n5 = LinearInterpolate(A=Constant() {r=1}, B=Constant() {r=0.3}, Alpha=Add(A=Divide(A=@n3, B=@n4), B=Constant() {r=0.5}))
  n6 = Multiply(
   A=TextureSampleParameter2D() {parameterName=color, texture=/Game/Environments/BuildingGuides/Textures/T_PBW_BG_BalconyO_QuarterWallS_QuarterWallHalf_BP_02_D}.RGB
   B=@n5
  )
  Add(A=Multiply(A=@n1, B=@n2.R), B=Add(A=Multiply(A=@n1, B=@n2.B), B=@n6))
  Multiply(A=Clamp(None=@n6), B=@n5)
 ]
}
