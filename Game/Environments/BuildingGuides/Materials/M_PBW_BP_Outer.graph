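# M_PBW_BP_Outer is an unlit, masked material, apparently for the outer part of building guide pieces, judging by its name and folder.
# An "alpha blend" texture parameter defaults to a balcony and quarter-wall guide mask, whose red channel is added to itself to double it.
# A standalone constant is also present; the graph shows no links from any node to the material outputs.
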Material /Game/Environments/BuildingGuides/Materials/M_PBW_BP_Outer
graph {
 properties {blendMode=BLEND_Masked, opacityMaskClipValue=0.3, pixelDepthOffsetMode=PDOM_Legacy, shadingModel=MSM_Unlit, usedShadingModels=Unlit}
 nodes [
  n1 = TextureSampleParameter2D() {parameterName=alpha blend, texture=/Game/Environments/BuildingGuides/Textures/T_PBW_BG_BalconyO_QuarterWallS_QuarterWallHalf_BP_01_M}
  Add(A=@n1.R, B=@n1.R)
  Constant() {r=2.25}
 ]
}
